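# Niagara effect for the Cyclone glider that draws a glowing fiery rail track with wheel sparks, fire sprites and rising embers, all sampled from the user skeletal mesh.
# Two rail ribbons interpolate between offset bone positions, while tie ribbons with lights loop along Y on engine time to fake motion.
# bisFrontEnd blends offsets, positions and spray directions by a sampled bone rotation.
# System LocalAlpha scales track opacity and spark and ember spawn rates.
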NiagaraSystem /BRCosmetics/Gadgets/Assets/VinderTech_GliderChute/Glider_Cyclone_Male/FX/NS_Glider_Cyclone
dependencies {
 usedDataInterfaces [&NiagaraDataInterfaceColorCurve, &NiagaraDataInterfaceCurve]
 usedDynamicInputs []
 usedModules [
  &SystemState
  &FN_Integrator
  &self:SystemSpawnScript.SystemScriptSource.SystemScriptGraph.NiagaraNodeAssignment_0.SetVariables_#1
 ]
}
emitters {
 Ash_Embers {
  simTarget=GPUComputeSim
  emitterUpdateScript [
   &EmitterState()
   &SpawnRate(
    SpawnRate=&Multiply_Float(A=&UniformRangedFloat(Minimum=0, Maximum=9), B=$System.LocalAlpha)
    Use Spawn Probability=false
   )
  ]
  particleSpawnScript [
   /Niagara/Modules/Spawn/Initialization/InitializeParticle(
    Color=<rgba 20 1.3333 0 1>
    Lifetime=&UniformRangedFloat(Minimum=0.5, Maximum=1.5)
    Module.Write Color=true
    Module.Write Lifetime=true
    Module.Write Mass=true
    Module.Write Material Random=true
    Module.Write Position=false
    Module.Write Scale=false
    Module.Write SpriteRotation=true
    Module.Write SpriteSize=true
    Position=&SimulationPosition()
    Sprite Rotation=&UniformRangedFloat(Minimum=0, Maximum=360)
    Sprite Size=&Vector2DFromFloat(Value=&UniformRangedFloat(Minimum=2, Maximum=5))
   )
   &SkeletalMeshLocation(
    Inherit Velocity=false
    MaskSpawnGroup=false
    Mesh Sampling Type=::NewEnumerator5
    Module.Override Randomness=false
    Skeletal Mesh=$User.SkeletalMesh
   )
   &SampleSkeletalMeshSkeleton(Mesh:NiagaraDataInterfaceSkeletalMesh=$User.SkeletalMesh, Specific Bone Index=5) {module=SampleSkeletalMeshSkeleton002}
   &AddVelocityInCone(
    Cone Axis=&Lerp_Vector(
     A=<0 1 0>
     B=&RotateVectorByQuaternion(Vector To Rotate=<0 1 0>, Quaternion=$Particles.SampleSkeletalMeshSkeleton002.SampledRotation)
     Alpha=&MakeCustomFloatFromBool(Bool=$User.bisFrontEnd, True Float=1, False Float=0)
    )
    Module.Use Velocity Falloff On Cone Axis=false
    Velocity Strength=&UniformRangedFloat(Minimum=150, Maximum=375)
   )
  ]
  particleUpdateScript [
   &ParticleState()
   &AccelerationForce(Acceleration=<0 0 500>)
   &Drag(Drag=&UniformRangedFloat(Minimum=1, Maximum=3), Use Linear Drag=true, Use Rotational Drag=false)
   &SolveForcesAndVelocity(Clamp Velocity=false, Limit Acceleration=false)
   &ScaleColor(
    Scale Alpha=&Lerp_Float(A=1, B=0, Alpha=$Particles.NormalizedAge)
    Scale RGB=&VectorFromFloat(Value=&FloatFromCurve(FloatCurve=DataInterface() {propertyValues {Curve {keys [
          {arriveTangent=-2886.0669, interpMode=RCIM_Cubic, leaveTangent=-2284.6055, tangentMode=RCTM_User, time=0, value=500}
          {interpMode=RCIM_Cubic, tangentMode=RCTM_User, time=0.3, value=0}
         ]}}}))
    ScaleA=true
    ScaleRGB=true
   )
  ]
  rendererClasses [&NiagaraSpriteRendererProperties]
  renderers [{rendererClass=&NiagaraSpriteRendererProperties}]
 }
 Fire_Decals {
  emitterUpdateScript [
   &EmitterState()
   &SpawnRate(SpawnRate=18, Use Spawn Probability=false)
  ]
  particleSpawnScript [
   &FN_SetSpawnGroup(Number of Spawn Groups=9)
   &InitializeParticle(
    Color=<rgba 5 1.25 0 0.125>
    Module.Write Color=true
    Module.Write Lifetime=true
    Module.Write Mass=true
    Module.Write Material Random=true
    Module.Write Position=false
    Module.Write Scale=false
    Module.Write SpriteRotation=true
    Module.Write SpriteSize=true
    Position=&SimulationPosition()
    Sprite Size=&Vector2DFromFloat(Value=&UniformRangedFloat(Minimum=3, Maximum=10))
   )
   &CameraOffset(Camera Offset Amount=20)
   /SetVariables(Particles.UVScale:Vector2f=&MakeVector2D(X=&UniformAOrBFloat(A=-1), Y=1)) {isSetParameters=true}
  ]
  particleUpdateScript [
   &ParticleState()
   &SkeletalMeshLocation(
    Bone / Socket Index=$Particles.SpawnGroup
    Inherit Velocity=false
    MaskSpawnGroup=false
    Mesh Sampling Type=::NewEnumerator5
    Skeletal Mesh=$User.SkeletalMesh
    Socket Sampling Mode=::NewEnumerator5
   )
   &SolveForcesAndVelocity(Clamp Velocity=false, Limit Acceleration=false)
   &ScaleColor(
    Scale Alpha=&FloatFromCurve(FloatCurve=DataInterface() {propertyValues {Curve {keys [<at 0 0 RCIM_Cubic>,<at 0.5 1 RCIM_Cubic>,<at 1 0 RCIM_Cubic>]}}})
    Scale RGB=&VectorFromFloat(Value=&FloatFromCurve(FloatCurve=DataInterface() {propertyValues {Curve {keys [<at 0 10>,<at 1 1>]}}}))
    ScaleA=true
    ScaleRGB=true
   )
   &ScaleSpriteSize(Scale Factor=&Vector2DFromFloat(Value=&Lerp_Float(A=1, B=2, Alpha=$Particles.NormalizedAge)))
  ]
  rendererClasses [&NiagaraSpriteRendererProperties]
  renderers [{rendererClass=&NiagaraSpriteRendererProperties}]
 }
 Rails {
  emitterUpdateScript [
   &EmitterState(
    Inactive Response=::NewEnumerator1
    Life Cycle Mode=::NewEnumerator1
    Loop Behavior=::NewEnumerator1
    Loop Duration Mode=::NewEnumerator1
    UseLoopDelay=false
   )
   &SpawnBurst_Instantaneous(Spawn Count=10, Use Spawn Probability=false, UseLoopCountLimit=false)
   &SpawnBurst_Instantaneous(Spawn Count=10, Spawn Group=1, Use Spawn Probability=false, UseLoopCountLimit=false) {enabled=false, module=SpawnBurst_Instantaneous001}
  ]
  particleSpawnScript [
   &FN_SetSpawnGroup(Normalize Spawn Group=true, Number of Spawn Groups=2, Particles Per Group=10, Set Ribbon ID and Ribbon Link Order=true) {enabled=false}
   /Niagara/Modules/Spawn/Initialization/V2/InitializeParticle(Lifetime=0, Ribbon Width=15, Ribbon Width Mode=::NewEnumerator1, UsePositionOffset=false, Write Material Random=true)
  ]
  particleUpdateScript [
   &ParticleState(Kill Particles When Lifetime Has Elapsed=false, Let Infinitely Lived Particles Die When Emitter Deactivates=true)
   &SkeletalMeshLocation(
    Bone Sampling Mode=::NewEnumerator4
    MaskSpawnGroup=false
    Orientation Sampling=::NewEnumerator1
    Position Sampling=::NewEnumerator2
    Sampled Position Offset=<0 -750 -15>
    Sampled Position Offset Coordinate Space:ENiagaraSkelMeshTransforms=::NewEnumerator2
    Skeletal Mesh=$User.SkeletalMesh
    Use Sampled Mesh Attribute as Ribbon ID=false
   )
   &SkeletalMeshLocation(
    Bone / Socket Index=2
    Bone Sampling Mode=::NewEnumerator4
    MaskSpawnGroup=false
    Orientation Sampling=::NewEnumerator1
    Position Sampling=::NewEnumerator2
    Sampled Position Offset=<0 750 -15>
    Sampled Position Offset Coordinate Space:ENiagaraSkelMeshTransforms=::NewEnumerator2
    Skeletal Mesh=$User.SkeletalMesh
    Use Sampled Mesh Attribute as Ribbon ID=false
   ) {module=SkeletalMeshLocation001}
   &SkeletalMeshLocation(
    Bone / Socket Index=5
    Bone Sampling Mode=::NewEnumerator4
    MaskSpawnGroup=false
    Orientation Sampling=::NewEnumerator1
    Position Sampling=::NewEnumerator2
    Skeletal Mesh=$User.SkeletalMesh
    Use Sampled Mesh Attribute as Ribbon ID=false
   ) {module=SkeletalMeshLocation002}
   /SetVariables(Particles.Position=&LerpPosition(
     A=$Particles.SkeletalMeshLocation.SampledPosition
     B=$Particles.SkeletalMeshLocation001.SampledPosition
     Alpha=&ReturnNormalizedExecIndex()
    )) {isSetParameters=true, module=SetVariables_#2}
   /SetVariables(Particles.MaterialRandom:NiagaraFloat=&MaskFloatBySpawnGroup(Masked Float=0.3, Passthrough Float=0)) {isSetParameters=true, module=SetVariables_#3}
   &DynamicMaterialParameters(
    Index 0 Param 2=&FloatFromCurve(
     FloatCurve=DataInterface() {propertyValues {Curve {keys [
         {interpMode=RCIM_Cubic, tangentMode=RCTM_User, time=0, value=0}
         {interpMode=RCIM_Cubic, tangentMode=RCTM_User, time=0.11, value=0}
         {arriveTangent=14.4915, interpMode=RCIM_Cubic, leaveTangent=-25.3485, tangentMode=RCTM_Break, time=0.2, value=0.5}
         {interpMode=RCIM_Cubic, tangentMode=RCTM_User, time=0.25, value=0}
         {interpMode=RCIM_Cubic, tangentMode=RCTM_User, time=0.35, value=0}
         {arriveTangent=15.8131, interpMode=RCIM_Cubic, leaveTangent=-25.9554, tangentMode=RCTM_Break, time=0.45, value=0.5}
         {interpMode=RCIM_Cubic, tangentMode=RCTM_User, time=0.5, value=0}
        ]}}}
     CurveIndex=&ReturnNormalizedExecIndex()
    )
    Index 0 Param 3=&FloatFromCurve(
     FloatCurve=DataInterface() {propertyValues {Curve {keys [
         <at 0 5 RCIM_Cubic>
         {arriveTangent=-25, interpMode=RCIM_Cubic, leaveTangent=-25, time=0.11, value=1.9759}
         {arriveTangent=0.1003, interpMode=RCIM_Cubic, leaveTangent=0.1003, time=0.2, value=0}
         {interpMode=RCIM_Cubic, tangentMode=RCTM_User, time=0.25, value=2}
         {interpMode=RCIM_Cubic, leaveTangent=-44.5746, tangentMode=RCTM_Break, time=0.35, value=2}
         <at 0.45 0 RCIM_Cubic>
         {arriveTangent=118.7072, interpMode=RCIM_Cubic, leaveTangent=5.4615, tangentMode=RCTM_Break, time=0.5, value=2}
         <at 1 3>
        ]}}}
     CurveIndex=&ReturnNormalizedExecIndex()
    )
    Module.Param0WriteEnabled=true
    Module.Param1WriteEnabled=true
    Module.Param2WriteEnabled=true
    Module.Param3WriteEnabled=true
   )
   &ScaleColor(
    Scale Alpha=$System.LocalAlpha
    Scale RGB=&MakeVectorFromLinearColorRGB(Color=&ColorFromCurve(
      ColorCurve=DataInterface() {propertyValues {
        AlphaCurve {keys [<at 0 1>,<at 1 1 RCIM_Cubic>,<at 1 1 RCIM_Cubic>]}
        BlueCurve {keys [<at 0 0>,<at 0.25 0>,<at 0.5 0 RCIM_Cubic>,<at 0.75 0 RCIM_Cubic>,<at 1 0 RCIM_Cubic>]}
        GreenCurve {keys [
          <at 0 0>
          <at 0.25 0.1667>
          <at 0.5 1.25 RCIM_Cubic>
          {interpMode=RCIM_Cubic, tangentMode=RCTM_User, time=0.75, value=0.1667}
          <at 1 0 RCIM_Cubic>
         ]}
        RedCurve {keys [<at 0 0>, <at 0.25 10>, <at 0.5 5 RCIM_Cubic>, {interpMode=RCIM_Cubic, tangentMode=RCTM_User, time=0.75, value=10}, <at 1 0.001 RCIM_Cubic>]}
       }}
      CurveIndex=&ReturnNormalizedExecIndex()
     ))
    ScaleA=true
    ScaleRGB=true
   )
   /SetVariables(Particles.Position=&RotateVectorByQuaternion(
     Vector To Rotate=Linked() {linkedVariable {name=Particles.Position, type=NiagaraPosition}}
     Quaternion=&LerpQuaternion(
      Quaternion A:Quat4f=<0 0 0 1>
      Quaternion B:Quat4f=$Particles.SkeletalMeshLocation002.SampledRotation
      Lerp Factor:NiagaraFloat=&MakeCustomFloatFromBool(Bool=$User.bisFrontEnd, True Float=1, False Float=0)
      Enable Spherical Linear Interpolation:NiagaraBool=false
     )
    )) {isSetParameters=true, module=SetVariables_#4}
   /SetVariables(
    Particles.RibbonID=/Niagara/DynamicInputs/Meshes/MakeRibbonIDFromSpawnGroup()
    Particles.RibbonLinkOrder:NiagaraFloat=&ReturnNormalizedExecIndex()
   ) {isSetParameters=true, module=SetVariables_#5}
  ]
  rendererClasses [&NiagaraRibbonRendererProperties]
  renderers [{rendererClass=&NiagaraRibbonRendererProperties}]
 }
 Rails001 {
  emitterUpdateScript [
   &EmitterState(
    Inactive Response=::NewEnumerator1
    Life Cycle Mode=::NewEnumerator1
    Loop Behavior=::NewEnumerator1
    Loop Duration Mode=::NewEnumerator1
    UseLoopDelay=false
   )
   &SpawnBurst_Instantaneous(Spawn Count=10, Use Spawn Probability=false, UseLoopCountLimit=false)
   &SpawnBurst_Instantaneous(Spawn Count=10, Spawn Group=1, Use Spawn Probability=false, UseLoopCountLimit=false) {enabled=false, module=SpawnBurst_Instantaneous001}
  ]
  particleSpawnScript [
   &FN_SetSpawnGroup(Normalize Spawn Group=true, Number of Spawn Groups=2, Particles Per Group=10, Set Ribbon ID and Ribbon Link Order=true) {enabled=false}
   &InitializeParticle(Lifetime=0, Ribbon Width=15, Ribbon Width Mode=::NewEnumerator1, UsePositionOffset=false, Write Material Random=true)
  ]
  particleUpdateScript [
   &ParticleState(Kill Particles When Lifetime Has Elapsed=false, Let Infinitely Lived Particles Die When Emitter Deactivates=true)
   &SkeletalMeshLocation(
    Bone / Socket Index=1
    Bone Sampling Mode=::NewEnumerator4
    MaskSpawnGroup=false
    Orientation Sampling=::NewEnumerator1
    Position Sampling=::NewEnumerator2
    Sampled Position Offset=<5 -750 -15>
    Sampled Position Offset Coordinate Space:ENiagaraSkelMeshTransforms=::NewEnumerator2
    Skeletal Mesh=$User.SkeletalMesh
    Use Sampled Mesh Attribute as Ribbon ID=false
   )
   &SkeletalMeshLocation(
    Bone / Socket Index=3
    Bone Sampling Mode=::NewEnumerator4
    MaskSpawnGroup=false
    Orientation Sampling=::NewEnumerator1
    Position Sampling=::NewEnumerator2
    Sampled Position Offset=<5 750 -15>
    Sampled Position Offset Coordinate Space:ENiagaraSkelMeshTransforms=::NewEnumerator2
    Skeletal Mesh=$User.SkeletalMesh
    Use Sampled Mesh Attribute as Ribbon ID=false
   ) {module=SkeletalMeshLocation001}
   &SkeletalMeshLocation(
    Bone / Socket Index=5
    Bone Sampling Mode=::NewEnumerator4
    MaskSpawnGroup=false
    Orientation Sampling=::NewEnumerator1
    Position Sampling=::NewEnumerator2
    Skeletal Mesh=$User.SkeletalMesh
    Use Sampled Mesh Attribute as Ribbon ID=false
   ) {module=SkeletalMeshLocation002}
   /SetVariables(Particles.Position=&LerpPosition(
     A=$Particles.SkeletalMeshLocation.SampledPosition
     B=$Particles.SkeletalMeshLocation001.SampledPosition
     Alpha=&ReturnNormalizedExecIndex()
    )) {isSetParameters=true, module=SetVariables_#2}
   /SetVariables(Particles.MaterialRandom:NiagaraFloat=&MaskFloatBySpawnGroup(Masked Float=0.3, Passthrough Float=0)) {isSetParameters=true, module=SetVariables_#3}
   &DynamicMaterialParameters(
    Index 0 Param 2=&FloatFromCurve(
     FloatCurve=DataInterface() {propertyValues {Curve {keys [
         {interpMode=RCIM_Cubic, tangentMode=RCTM_User, time=0, value=0}
         {interpMode=RCIM_Cubic, tangentMode=RCTM_User, time=0.11, value=0}
         {arriveTangent=14.4915, interpMode=RCIM_Cubic, leaveTangent=-25.3485, tangentMode=RCTM_Break, time=0.2, value=0.5}
         {interpMode=RCIM_Cubic, tangentMode=RCTM_User, time=0.25, value=0}
         {interpMode=RCIM_Cubic, tangentMode=RCTM_User, time=0.35, value=0}
         {arriveTangent=15.8131, interpMode=RCIM_Cubic, leaveTangent=-25.9554, tangentMode=RCTM_Break, time=0.45, value=0.5}
         {interpMode=RCIM_Cubic, tangentMode=RCTM_User, time=0.5, value=0}
        ]}}}
     CurveIndex=&ReturnNormalizedExecIndex()
    )
    Index 0 Param 3=&FloatFromCurve(
     FloatCurve=DataInterface() {propertyValues {Curve {keys [
         <at 0 5 RCIM_Cubic>
         {arriveTangent=-25, interpMode=RCIM_Cubic, leaveTangent=-25, time=0.11, value=1.9759}
         {arriveTangent=0.1003, interpMode=RCIM_Cubic, leaveTangent=0.1003, time=0.2, value=0}
         {interpMode=RCIM_Cubic, tangentMode=RCTM_User, time=0.25, value=2}
         {interpMode=RCIM_Cubic, leaveTangent=-44.5746, tangentMode=RCTM_Break, time=0.35, value=2}
         <at 0.45 0 RCIM_Cubic>
         {arriveTangent=118.7072, interpMode=RCIM_Cubic, leaveTangent=5.4615, tangentMode=RCTM_Break, time=0.5, value=2}
         <at 1 3>
        ]}}}
     CurveIndex=&ReturnNormalizedExecIndex()
    )
    Module.Param0WriteEnabled=true
    Module.Param1WriteEnabled=true
    Module.Param2WriteEnabled=true
    Module.Param3WriteEnabled=true
   )
   &ScaleColor(
    Scale Alpha=$System.LocalAlpha
    Scale RGB=&MakeVectorFromLinearColorRGB(Color=&ColorFromCurve(
      ColorCurve=DataInterface() {propertyValues {
        AlphaCurve {keys [<at 0 1>,<at 1 1 RCIM_Cubic>,<at 1 1 RCIM_Cubic>]}
        BlueCurve {keys [<at 0 0>,<at 0.25 0>,<at 0.5 0 RCIM_Cubic>,<at 0.75 0 RCIM_Cubic>,<at 1 0 RCIM_Cubic>]}
        GreenCurve {keys [
          <at 0 0>
          <at 0.25 0.1667>
          <at 0.5 1.25 RCIM_Cubic>
          {interpMode=RCIM_Cubic, tangentMode=RCTM_User, time=0.75, value=0.1667}
          <at 1 0 RCIM_Cubic>
         ]}
        RedCurve {keys [<at 0 0>, <at 0.25 10>, <at 0.5 5 RCIM_Cubic>, {interpMode=RCIM_Cubic, tangentMode=RCTM_User, time=0.75, value=10}, <at 1 0.001 RCIM_Cubic>]}
       }}
      CurveIndex=&ReturnNormalizedExecIndex()
     ))
    ScaleA=true
    ScaleRGB=true
   )
   /SetVariables(Particles.Position=&RotateVectorByQuaternion(
     Vector To Rotate=Linked() {linkedVariable {name=Particles.Position, type=NiagaraPosition}}
     Quaternion=&LerpQuaternion(
      Quaternion A:Quat4f=<0 0 0 1>
      Quaternion B:Quat4f=$Particles.SkeletalMeshLocation002.SampledRotation
      Lerp Factor:NiagaraFloat=&MakeCustomFloatFromBool(Bool=$User.bisFrontEnd, True Float=1, False Float=0)
      Enable Spherical Linear Interpolation:NiagaraBool=false
     )
    )) {isSetParameters=true, module=SetVariables_#4}
   /SetVariables(
    Particles.RibbonID=&MakeRibbonIDFromSpawnGroup()
    Particles.RibbonLinkOrder:NiagaraFloat=&ReturnNormalizedExecIndex()
   ) {isSetParameters=true, module=SetVariables_#5}
  ]
  rendererClasses [&NiagaraRibbonRendererProperties]
  renderers [{rendererClass=&NiagaraRibbonRendererProperties}]
 }
 Ties {
  emitterUpdateScript [
   &EmitterState(
    Inactive Response=::NewEnumerator1
    Life Cycle Mode=::NewEnumerator1
    Loop Behavior=::NewEnumerator1
    Loop Duration Mode=::NewEnumerator1
    UseLoopDelay=false
   )
   &SpawnBurst_Instantaneous(Spawn Count=55, Use Spawn Probability=false)
  ]
  particleSpawnScript [
   &FN_SetSpawnGroup(Number of Spawn Groups=11)
   &InitializeRibbon(
    Color=<rgba 1 1 1 0.5>
    Module.Write Color=true
    Module.Write Lifetime=true
    Module.Write Mass=true
    Module.Write Material Random=true
    Module.Write Position=false
    Module.Write RibbonTwist=false
    Module.Write RibbonWidth=true
    Position:Vector3f=&SimulationPosition()
    Ribbon Width=30
   )
   /SetVariables(Particles.Alpha2=HlslExpression() {hlslExpression="float(Modulo(ExecIndex(), 5)) / 4.0"}) {isSetParameters=true}
   /SetVariables(
    Particles.LightEnablaled:NiagaraBool=&Boolean_AndOperation(
     Bool1=&InvertBool(Boolean:NiagaraBool=/Niagara/DynamicInputs/Bool/SetBoolByFloatComparison(A:NiagaraFloat=$Particles.Alpha2, B:NiagaraFloat=0.5, ComparisonType:ENiagaraBooleanLogicOps=::NewEnumerator0))
     Bool2=&SetBoolByFloatComparison(A:NiagaraFloat=$Particles.Alpha2, B:NiagaraFloat=0.4, ComparisonType:ENiagaraBooleanLogicOps=::NewEnumerator0)
    )
    Particles.LightExponent=6
   ) {isSetParameters=true}
  ]
  particleUpdateScript [
   &ParticleState(Kill Particles When Lifetime Has Elapsed=false, Let Infinitely Lived Particles Die When Emitter Deactivates=true)
   &SampleSkeletalMeshSkeleton(Mesh:NiagaraDataInterfaceSkeletalMesh=$User.SkeletalMesh, Specific Bone Index=0)
   &SampleSkeletalMeshSkeleton(Mesh:NiagaraDataInterfaceSkeletalMesh=$User.SkeletalMesh, Specific Bone Index=1) {module=SampleSkeletalMeshSkeleton001}
   &SampleSkeletalMeshSkeleton(Mesh:NiagaraDataInterfaceSkeletalMesh=$User.SkeletalMesh, Specific Bone Index=5) {module=SampleSkeletalMeshSkeleton002}
   /SetVariables(
    Particles.Position=&Lerp_Vector(
     A=&MakeVector(
      X=&MakeFloatFromVector(Vector=$Particles.SampleSkeletalMeshSkeleton.SampledPosition, Channel=::NewEnumerator0)
      Y=&Lerp_Float(
       A=&MakeFloatFromVector(Vector=$Particles.SampleSkeletalMeshSkeleton.SampledPosition, Channel=::NewEnumerator1)
       B=&Add_Float(
        A=&MakeFloatFromVector(Vector=$Particles.SampleSkeletalMeshSkeleton.SampledPosition, Channel=::NewEnumerator1)
        B=100
       )
       Alpha=&MakeFloatFromInt(Int=$Particles.SpawnGroup)
      )
      Z=&MakeFloatFromVector(Vector=$Particles.SampleSkeletalMeshSkeleton.SampledPosition, Channel=::NewEnumerator2)
     )
     B=&MakeVector(
      X=&MakeFloatFromVector(Vector=$Particles.SampleSkeletalMeshSkeleton001.SampledPosition, Channel=::NewEnumerator0)
      Y=&Lerp_Float(
       A=&MakeFloatFromVector(Vector=$Particles.SampleSkeletalMeshSkeleton001.SampledPosition, Channel=::NewEnumerator1)
       B=&Add_Float(
        A=&MakeFloatFromVector(Vector=$Particles.SampleSkeletalMeshSkeleton001.SampledPosition, Channel=::NewEnumerator1)
        B=100
       )
       Alpha=&MakeFloatFromInt(Int=$Particles.SpawnGroup)
      )
      Z=&MakeFloatFromVector(Vector=$Particles.SampleSkeletalMeshSkeleton001.SampledPosition, Channel=::NewEnumerator2)
     )
     Alpha=$Particles.Alpha2
    )
    Particles.RibbonID=/Niagara/DynamicInputs/NiagaraID/MakeNiagaraIDFromInt(Index:NiagaraInt32=$Particles.SpawnGroup, Acquire Tag:NiagaraInt32=0)
   ) {isSetParameters=true}
   /SetVariables(Particles.Position=&Add_Vector(
     A=Linked() {linkedVariable {name=Particles.Position, type=NiagaraPosition}}
     B=&Lerp_Vector(
      A=&Lerp_Vector(
       A=<-60 0 -15>
       B=&RotateVectorByQuaternion(Vector To Rotate=<-60 0 -15>, Quaternion=$Particles.SampleSkeletalMeshSkeleton002.SampledRotation)
       Alpha=&MakeCustomFloatFromBool(Bool=$User.bisFrontEnd, True Float=1, False Float=0)
      )
      B=&Lerp_Vector(
       A=<60 0 -15>
       B=&RotateVectorByQuaternion(Vector To Rotate=<60 0 -15>, Quaternion=$Particles.SampleSkeletalMeshSkeleton002.SampledRotation)
       Alpha=&MakeCustomFloatFromBool(Bool=$User.bisFrontEnd, True Float=1, False Float=0)
      )
      Alpha=$Particles.Alpha2
     )
    )) {isSetParameters=true}
   /SetVariables(Particles.Position=&Add_Vector(
     A=Linked() {linkedVariable {name=Particles.Position, type=NiagaraPosition}}
     B=&Lerp_Vector(
      A=<0 -550 0>
      B=<0 -350 0>
      Alpha=&Frac_Float(Float=&Multiply_Float(A=$Engine.Time, B=6))
     )
    )) {isSetParameters=true}
   &DynamicMaterialParameters(
    Index 0 Param 2=1
    Module.Param0WriteEnabled=false
    Module.Param1WriteEnabled=true
    Module.Param2WriteEnabled=true
    Module.Param3WriteEnabled=false
   ) {module=DynamicMaterialParameters001}
   &ScaleColor(
    Scale Alpha=&FloatFromCurve(
     FloatCurve=DataInterface() {propertyValues {Curve {keys [
         {interpMode=RCIM_Cubic, tangentMode=RCTM_Break, time=0, value=0}
         {arriveTangent=3.899, interpMode=RCIM_Cubic, leaveTangent=-6.3004, tangentMode=RCTM_Break, time=0.5, value=1}
         {interpMode=RCIM_Cubic, tangentMode=RCTM_Break, time=1, value=0}
        ]}}}
     CurveIndex=&Divide_Float(A=&MakeFloatFromInt(Int=$Particles.SpawnGroup), B=10)
    )
    Scale RGB=&MakeVectorFromLinearColorRGB(Color=&ColorFromCurve(
      ColorCurve=DataInterface() {propertyValues {
        AlphaCurve {keys [<at 0 1 RCIM_Cubic>,<at 1 1 RCIM_Cubic>,<at 1 1 RCIM_Cubic>]}
        BlueCurve {keys [<at 0 0>,<at 0.25 0>,<at 0.5 0 RCIM_Cubic>,<at 0.75 0 RCIM_Cubic>,<at 1 0 RCIM_Cubic>]}
        GreenCurve {keys [
          <at 0 0>
          <at 0.25 0.1667>
          {arriveTangent=-0.2222, interpMode=RCIM_Cubic, leaveTangent=-0.2222, time=0.5, value=1.25}
          {interpMode=RCIM_Cubic, tangentMode=RCTM_User, time=0.75, value=0.1667}
          <at 1 0 RCIM_Cubic>
         ]}
        RedCurve {keys [
          <at 0 0.001>
          <at 0.25 10>
          {arriveTangent=-13.332, interpMode=RCIM_Cubic, leaveTangent=-13.332, time=0.5, value=5}
          {interpMode=RCIM_Cubic, tangentMode=RCTM_User, time=0.75, value=10}
          <at 1 0.001 RCIM_Cubic>
         ]}
       }}
      CurveIndex=$Particles.Alpha2
     ))
    ScaleA=true
    ScaleRGB=true
   ) {module=ScaleColor001}
   &ScaleColor(Scale Alpha=$System.LocalAlpha, Scale RGB=<5 5 5>, ScaleA=true, ScaleRGB=true)
   &SolveForcesAndVelocity(Clamp Velocity=false, Limit Acceleration=false)
  ]
  rendererClasses [&NiagaraRibbonRendererProperties, &NiagaraLightRendererProperties]
  renderers [
   {rendererClass=&NiagaraRibbonRendererProperties}
   {rendererClass=&NiagaraLightRendererProperties, rendererIndex=1}
  ]
 }
 WheelSparks {
  emitterUpdateScript [
   &EmitterState()
   &SpawnRate(
    SpawnRate=&Multiply_Float(A=&RandomRangeFloat(Minimum=0, Maximum=300), B=$System.LocalAlpha)
    Use Spawn Probability=false
   )
  ]
  particleSpawnScript [
   &FN_SetSpawnGroup()
   /Niagara/Modules/Spawn/Initialization/InitializeParticle(
    Color=&RandomRangeLinearColor(Minimum=<rgba 1 0.6167 0 1>, Maximum=<rgba 1 0.0167 0 1>)
    Lifetime=&UniformRangedFloat(Minimum=0.05, Maximum=0.3)
    Module.Write Color=true
    Module.Write Lifetime=true
    Module.Write Mass=true
    Module.Write Material Random=true
    Module.Write Position=false
    Module.Write Scale=false
    Module.Write SpriteRotation=false
    Module.Write SpriteSize=true
    Position=&SimulationPosition()
    Sprite Size=&Vector2DFromFloat(Value=&UniformRangedFloat(Minimum=2, Maximum=3))
   )
   &SkeletalMeshLocation(
    Bone / Socket Index=$Particles.SpawnGroup
    Bone Sampling Mode=::NewEnumerator4
    Inherit Velocity=false
    MaskSpawnGroup=false
    Sampled Position Offset=&Lerp_Vector(A=<-7 0 -15>, B=<7 0 -15>, Alpha=HlslExpression() {hlslExpression="float(Particles.SpawnGroup && 1)"})
    Skeletal Mesh=$User.SkeletalMesh
   )
   &SampleSkeletalMeshSkeleton(Mesh:NiagaraDataInterfaceSkeletalMesh=$User.SkeletalMesh, Specific Bone Index=5) {module=SampleSkeletalMeshSkeleton002}
   &BoxLocation(Box Size=<10 10 10>, MaskSpawnGroup=false, Module.Override Seed=false, Rotate Box=false, Surface Only=false)
   &AddVelocityInCone(
    Cone Axis=&Lerp_Vector(
     A=&Lerp_Vector(A=<-1 1 0.5>, B=<1 1 0.5>, Alpha=HlslExpression() {hlslExpression="float(Particles.SpawnGroup && 1)"})
     B=&RotateVectorByQuaternion(
      Vector To Rotate=&Lerp_Vector(A=<-1 1 0.5>, B=<1 1 0.5>, Alpha=HlslExpression() {hlslExpression="float(Particles.SpawnGroup && 1)"})
      Quaternion=$Particles.SampleSkeletalMeshSkeleton002.SampledRotation
     )
     Alpha=&MakeCustomFloatFromBool(Bool=$User.bisFrontEnd, True Float=1, False Float=0)
    )
    Module.Use Velocity Falloff On Cone Axis=false
    Velocity Strength=500
   )
  ]
  particleUpdateScript [
   &ParticleState()
   &AccelerationForce(Acceleration=<0 0 -750>)
   &Drag(Drag=&UniformRangedFloat(Minimum=1, Maximum=3), Use Linear Drag=true, Use Rotational Drag=false)
   &SolveForcesAndVelocity(Clamp Velocity=false, Limit Acceleration=false)
   &ScaleSpriteSizeBySpeed(Max Scale Factor=<0.5 4>, Min Scale Factor=<0.25 1>)
   &ScaleColor(
    Scale RGB=&VectorFromFloat(Value=&Lerp_Float(A=100, B=1, Alpha=&Power(a=$Particles.NormalizedAge, b=0.2)))
    ScaleA=true
    ScaleRGB=true
   )
   &DynamicMaterialParameters(
    Index 0 Param 1=&FloatFromCurve(FloatCurve=DataInterface() {propertyValues {Curve {keys [<at 0 5>,<at 1 1>]}}})
    Index 0 Param 2=5
    Index 0 Param 3=10
    Module.Param0WriteEnabled=true
    Module.Param1WriteEnabled=true
    Module.Param2WriteEnabled=true
    Module.Param3WriteEnabled=true
   )
  ]
  rendererClasses [&NiagaraSpriteRendererProperties]
  renderers [{rendererClass=&NiagaraSpriteRendererProperties}]
 }
}
summary {
 systemName=NS_Glider_Cyclone
 emitters [
  {emitterName=Ties, rendererClasses [&NiagaraRibbonRendererProperties, &NiagaraLightRendererProperties]}
  {emitterName=WheelSparks, rendererClasses [&NiagaraSpriteRendererProperties]}
  {emitterName=Fire_Decals, rendererClasses [&NiagaraSpriteRendererProperties]}
  {emitterName=Ash_Embers, simTarget=GPUComputeSim, rendererClasses [&NiagaraSpriteRendererProperties]}
  {emitterName=Rails, rendererClasses [&NiagaraRibbonRendererProperties]}
  {emitterName=Rails001, rendererClasses [&NiagaraRibbonRendererProperties]}
 ]
 userVariables [
  User.bisFrontEnd {type=NiagaraBool}
  User.bIsFrontEndPreview {type=NiagaraBool}
  User.bisFullyDeployed {type=NiagaraBool, defaultValue=true}
  User.MinDistance {type=NiagaraFloat, defaultValue=0, description=""}
  User.RibbonIDOffset {type=NiagaraInt32}
  User.SkeletalMesh {
   type=NiagaraDataInterfaceSkeletalMesh
   defaultValue {
    struct=NiagaraExt_VariableValue_DataInterface
    value {dataInterface=&self:NiagaraDataInterfaceSkeletalMesh_1}
   }
  }
 ]
}
