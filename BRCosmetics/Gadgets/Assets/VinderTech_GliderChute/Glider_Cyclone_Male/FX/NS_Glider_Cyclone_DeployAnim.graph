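# Niagara effect for the Cyclone glider's deploy animation, rendering an anim-trail ribbon from a skeletal mesh.
# A single FN_AnimTrail_V4 emitter spawns one trail from the user skeletal mesh through shared FN spawn, initialize and update modules using bone or socket indices.
# Alpha follows a low-exponent power curve over normalized age, dropping sharply early and tailing off over each particle's lifetime.
# The ribbon-facing override and material module are disabled.
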NiagaraSystem /BRCosmetics/Gadgets/Assets/VinderTech_GliderChute/Glider_Cyclone_Male/FX/NS_Glider_Cyclone_DeployAnim
dependencies {usedDataInterfaces [], usedDynamicInputs [], usedModules [&SystemState]}
emitters {FN_AnimTrail_V4 {
  emitterUpdateScript [
   &EmitterState()
   /Game/Effects/Niagara/Modules/AnimTrail/FN_SpawnAnimTrail/FN_SpawnAnimTrail(Number of Anim Trails=1, Spawn Mode:ENiagaraSkeletalSpawning=::NewEnumerator0)
  ]
  particleSpawnScript [
   /Game/Effects/Niagara/Modules/AnimTrail/FN_InitializeAnimTrail/FN_InitializeAnimTrailBoneIndecies(Second Bone or Socket Index=1)
   /Game/Effects/Niagara/Modules/AnimTrail/FN_InitializeAnimTrail/FN_InitializeAnimTrail(AnimTrailWidth=1, Lifetime=1.5, Spawn Mode:ENiagaraSkeletalSpawning=::NewEnumerator0, Write Material Random=false)
  ]
  particleUpdateScript [
   &SolveForcesAndVelocity(Clamp Velocity=false, Limit Acceleration=false)
   /Game/Effects/Niagara/Modules/AnimTrail/FN_UpdateAnimTrail/FN_UpdateAnimTrail()
   /SetVariables(Particles.RibbonFacing=&TransformVector(Vector=<0 0 1>, Source Space=::World, Destination Space=::Simulation)) {enabled=false, isSetParameters=true}
   /Game/Effects/Niagara/Modules/Materials/Material_v2() {enabled=false}
   &ScaleColor(
    Scale Alpha=&Lerp_Float(A=1, B=0, Alpha=&Power(a=$Particles.NormalizedAge, b=0.1))
    ScaleA=true
    ScaleRGB=true
   )
  ]
  rendererClasses [&NiagaraRibbonRendererProperties]
  renderers [{rendererClass=&NiagaraRibbonRendererProperties}]
 }}
summary {
 systemName=NS_Glider_Cyclone_DeployAnim
 emitters [{emitterName=FN_AnimTrail_V4, rendererClasses [&NiagaraRibbonRendererProperties]}]
 userVariables [
  User.RibbonIDOffset {type=NiagaraInt32}
  User.SkeletalMesh {
   type=NiagaraDataInterfaceSkeletalMesh
   defaultValue {
    struct=NiagaraExt_VariableValue_DataInterface
    value {dataInterface=&self:NiagaraDataInterfaceSkeletalMesh_0}
   }
  }
 ]
}
