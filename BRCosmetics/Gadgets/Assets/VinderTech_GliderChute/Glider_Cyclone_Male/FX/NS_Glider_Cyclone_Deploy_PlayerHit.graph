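# Niagara effect for the Cyclone glider, named for a deploy player hit, that throws brief streaks and sparks from bones of User.SkeletalMesh.
# Both emitters map a spawn group to a bone index and lerp offset and cone direction on `SpawnGroup && 1`, mirroring group 0 against the others.
# Streaks bursts randomly tinted HDR sprites that stretch while fading.
# GPU WheelSparks adds burst plus rate-spawned sparks with gravity, drag and speed-based stretching.
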NiagaraSystem /BRCosmetics/Gadgets/Assets/VinderTech_GliderChute/Glider_Cyclone_Male/FX/NS_Glider_Cyclone_Deploy_PlayerHit
dependencies {
 usedDataInterfaces [&NiagaraDataInterfaceCurve, &NiagaraDataInterfaceColorCurve]
 usedDynamicInputs []
 usedModules [&SystemState]
}
emitters {
 Streaks {
  emitterUpdateScript [
   &EmitterState()
   &SpawnBurst_Instantaneous(Spawn Count=20, Use Spawn Probability=false)
  ]
  particleSpawnScript [
   /Niagara/Modules/Spawn/Initialization/InitializeParticle(
    Color=&MakeLinearColorFromVectorAndFloat(
     "Vector (RGB)"=&Multiply_Vector(
      A=&MakeVectorFromLinearColorRGB(Color=&ColorFromCurve(
        ColorCurve=DataInterface() {propertyValues {
          AlphaCurve {keys [<at 0 1>,<at 1 1>]}
          BlueCurve {keys [
            {interpMode=RCIM_Cubic, leaveTangent=0.1119, tangentMode=RCTM_Break, time=0, value=0.3333}
            {arriveTangent=-0.6438, interpMode=RCIM_Cubic, leaveTangent=-0.6438, time=0.3073, value=0.599}
            {arriveTangent=-1.6357, interpMode=RCIM_Cubic, leaveTangent=-1.6357, time=0.5178, value=0}
            {interpMode=RCIM_Cubic, tangentMode=RCTM_Break, time=0.6364, value=0.0606}
            {interpMode=RCIM_Cubic, tangentMode=RCTM_Break, time=1, value=0.349}
           ]}
          GreenCurve {keys [
            {interpMode=RCIM_Cubic, tangentMode=RCTM_Break, time=0, value=0.0901}
            {arriveTangent=0.369, interpMode=RCIM_Cubic, leaveTangent=0.369, tangentMode=RCTM_User, time=0.3073, value=0.148}
            {arriveTangent=0.0694, interpMode=RCIM_Cubic, leaveTangent=0.0694, tangentMode=RCTM_User, time=0.5178, value=0.2908}
            {arriveTangent=-0.3542, interpMode=RCIM_Cubic, leaveTangent=-0.3542, tangentMode=RCTM_User, time=0.6364, value=0}
            {arriveTangent=0.0185, interpMode=RCIM_Cubic, leaveTangent=0.0185, tangentMode=RCTM_User, time=1, value=0}
           ]}
          RedCurve {keys [
            <at 0 0.0156 RCIM_Cubic>
            {
             arriveTangent=0.3413
             arriveTangentWeight=0.4006
             interpMode=RCIM_Cubic
             leaveTangent=0.3413
             leaveTangentWeight=0.4006
             tangentMode=RCTM_Break
             time=0.3073
             value=0.0034
            }
            {arriveTangent=4.0672, interpMode=RCIM_Cubic, leaveTangent=-0.5959, tangentMode=RCTM_Break, time=0.5178, value=1}
            {arriveTangent=-3.2078, interpMode=RCIM_Cubic, leaveTangent=-3.2078, tangentMode=RCTM_User, time=0.6364, value=0.2135}
            <at 1 0.2649 RCIM_Cubic>
           ]}
         }}
        CurveIndex=&UniformRangedFloat(Minimum=0.38, Maximum=0.6)
       ))
      B=&VectorFromFloat(Value=&UniformRangedFloat(Minimum=10, Maximum=100))
     )
     "Float (Alpha)"=1
    )
    Lifetime=&UniformRangedFloat(Minimum=0.075, Maximum=0.3)
    Module.Write Color=true
    Module.Write Lifetime=true
    Module.Write Mass=true
    Module.Write Material Random=true
    Module.Write Position=false
    Module.Write Scale=false
    Module.Write SpriteRotation=false
    Module.Write SpriteSize=true
    Position=&SimulationPosition()
    Sprite Size=&UniformRangedVector2D(Minimum=<20 80>, Maximum=<40 100>)
   )
   /Game/Effects/Niagara/Modules/Groups/FN_SetSpawnGroup_v1/FN_SetSpawnGroup(Emitter Bone Count=4)
   &SampleSkeletalMeshSkeleton(
    Mesh:NiagaraDataInterfaceSkeletalMesh=$User.SkeletalMesh
    Sampled Position Coordinate Space:ENiagaraCoordinateSpace=::Local
    Sampled Rotation Coordinate Space:ENiagaraCoordinateSpace=::Local
    Sampled Velocity Coordinate Space:ENiagaraCoordinateSpace=::Local
    Specific Bone Index:NiagaraInt32=$Particles.SpawnGroup
   )
   /Niagara/Modules/Spawn/Location/SkeletalMeshSkeletonLocation(
    Module.Inherit Velocity=false
    Module.Use Sampled Bone as Ribbon ID=true
    Sampled Position Offset:Vector3f=&Lerp_Vector(A=<-7 0 -15>, B=<7 0 -15>, Alpha=HlslExpression() {hlslExpression="float(Particles.SpawnGroup && 1)"})
    Sampled Rotation Attribute:Quat4f=<0 0 0 0>
   )
   &AddVelocityInCone(
    Cone Axis=&Lerp_Vector(A=<1 0 0.5>, B=<-1 0 0.5>, Alpha=HlslExpression() {hlslExpression="float(Particles.SpawnGroup && 1)"})
    Module.Use Velocity Falloff On Cone Axis=false
    Velocity Strength=10
   )
   /SetVariables(Particles.SubImageIndex=&UniformRangedFloat(Minimum=0, Maximum=3)) {isSetParameters=true}
  ]
  particleUpdateScript [
   &ParticleState()
   &SolveForcesAndVelocity(Clamp Velocity=false, Limit Acceleration=false)
   &ScaleSpriteSize(Scale Factor=&MakeVector2D(
     X=&Lerp_Float(A=0.75, B=1, Alpha=$Particles.NormalizedAge)
     Y=&Lerp_Float(A=0.03, B=2, Alpha=&Power(a=$Particles.NormalizedAge, b=5))
    ))
   &ScaleColor(
    Scale Alpha=&Lerp_Float(A=1, B=0, Alpha=$Particles.NormalizedAge)
    Scale RGB=&Lerp_Vector(A=<10 10 10>, B=<1 1 1>, Alpha=$Particles.NormalizedAge)
    ScaleA=true
    ScaleRGB=true
   )
  ]
  rendererClasses [&NiagaraSpriteRendererProperties]
  renderers [{rendererClass=&NiagaraSpriteRendererProperties}]
 }
 WheelSparks {
  simTarget=GPUComputeSim
  emitterUpdateScript [
   &EmitterState()
   &SpawnBurst_Instantaneous(Spawn Count=&UniformRangedInt(Minimum=100, Maximum=200), Use Spawn Probability=false)
   &SpawnRate(SpawnRate=400, Use Spawn Probability=false)
  ]
  particleSpawnScript [
   &InitializeParticle(
    Color=<rgba 10 10 10 1>
    Lifetime=&UniformRangedFloat(Minimum=0.05, Maximum=0.3)
    Module.Write Color=true
    Module.Write Lifetime=true
    Module.Write Mass=true
    Module.Write Material Random=true
    Module.Write Position=false
    Module.Write Scale=false
    Module.Write SpriteRotation=false
    Module.Write SpriteSize=true
    Position=&SimulationPosition()
    Sprite Size=&Vector2DFromFloat(Value=&UniformRangedFloat(Minimum=3, Maximum=5))
   )
   &FN_SetSpawnGroup(Emitter Bone Count=4)
   &SampleSkeletalMeshSkeleton(
    Mesh:NiagaraDataInterfaceSkeletalMesh=$User.SkeletalMesh
    Sampled Position Coordinate Space:ENiagaraCoordinateSpace=::Local
    Sampled Rotation Coordinate Space:ENiagaraCoordinateSpace=::Local
    Sampled Velocity Coordinate Space:ENiagaraCoordinateSpace=::Local
    Specific Bone Index:NiagaraInt32=$Particles.SpawnGroup
   )
   &SkeletalMeshSkeletonLocation(
    Module.Inherit Velocity=false
    Module.Use Sampled Bone as Ribbon ID=true
    Sampled Position Offset:Vector3f=&Lerp_Vector(A=<-7 0 -15>, B=<7 0 -15>, Alpha=HlslExpression() {hlslExpression="float(Particles.SpawnGroup && 1)"})
    Sampled Rotation Attribute:Quat4f=<0 0 0 0>
   )
   &BoxLocation(Box Size=<10 10 10>, MaskSpawnGroup=false, Module.Override Seed=false, Rotate Box=false, Surface Only=false)
   &AddVelocityInCone(
    Cone Angle=60
    Cone Axis=&Lerp_Vector(A=<1 0 0.5>, B=<-1 0 0.5>, Alpha=HlslExpression() {hlslExpression="float(Particles.SpawnGroup && 1)"})
    Module.Use Velocity Falloff On Cone Axis=false
    Velocity Strength=750
   )
  ]
  particleUpdateScript [
   &ParticleState()
   &AccelerationForce(Acceleration=<0 0 -750>)
   &Drag(Drag=&UniformRangedFloat(Minimum=1, Maximum=3), Use Linear Drag=true, Use Rotational Drag=false)
   &SolveForcesAndVelocity(Clamp Velocity=false, Limit Acceleration=false)
   &ScaleSpriteSizeBySpeed(Max Scale Factor=<1 8>, Min Scale Factor=<0.25 1>)
   &ScaleColor(
    Scale RGB=&VectorFromFloat(Value=&Lerp_Float(A=100, B=1, Alpha=&Power(a=$Particles.NormalizedAge, b=0.2)))
    ScaleA=true
    ScaleRGB=true
   )
  ]
  rendererClasses [&NiagaraSpriteRendererProperties]
  renderers [{rendererClass=&NiagaraSpriteRendererProperties}]
 }
}
summary {
 systemName=NS_Glider_Cyclone_Deploy_PlayerHit
 emitters [
  {emitterName=WheelSparks, simTarget=GPUComputeSim, rendererClasses [&NiagaraSpriteRendererProperties]}
  {emitterName=Streaks, rendererClasses [&NiagaraSpriteRendererProperties]}
 ]
 userVariables [
  User.LocalAlpha {type=NiagaraFloat, defaultValue=0, description=""}
  User.SkeletalMesh {
   type=NiagaraDataInterfaceSkeletalMesh
   defaultValue {
    struct=NiagaraExt_VariableValue_DataInterface
    value {dataInterface=&self:NiagaraDataInterfaceSkeletalMesh_0}
   }
  }
 ]
}
